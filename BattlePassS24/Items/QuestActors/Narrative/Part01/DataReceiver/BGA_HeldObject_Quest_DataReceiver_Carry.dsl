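; Carryable Data Receiver for a Season 24 narrative quest, which its owning player picks up on spawn and can retrieve after dropping.
; FortHeldObject events toggle Held, icons and owner interaction, and release removes the holding effect.
; OnRep_Held sends pickup/drop gameplay events and runs a timer checking capsule overlap with the placement actor.
; On overlap it drops, hides its mesh and reports quest progress; some logic retains sign and bear names.
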(event OnHeldObjectPlaced(FortHeldObject)
  ;[1 1 1] When we set the sign down, if it is overlapping with a player, push the sign away from the direction the player is facing
  (Utilities|Array|ForEachLoopwithBreak (Collision|GetOverlappingActors (Variables|Default|GetCapsule) "/Script/FortniteGame.FortPlayerPawn" :!class /Script/Engine.PrimitiveComponent)
    (:LoopBody
      (Variables|/Script/Engine.MovementComponent|SetVelocity (* (+ (* (Math|Vector|GetForwardVector (Transformation|GetActorRotation (HeldObjects|Instigator|GetInstigator (Variables|Default|GetFortHeldObject)))) -1.0) "0, 0,6.000000") 2e+02) (Variables|Default|GetFortProjectileMovement)))
    (:Completed
      _))
  ;[1 1 1] Set variables to track player interaction; report quest progress
  (Variables|Default|SetHeld false)
  (Transformation|SetActorRotation (Transformation|GetActorRotation (HeldObjects|Instigator|GetInstigator (Variables|Default|GetFortHeldObject))))
  (CallFunction|ControlIcons true)
  (GameplayEffects|RemoveActiveEffectswithGrantedTags (Ability|GetAbilitySystemComponent (Actor|GetOwner)) "(GameplayTags=((TagName=\"Athena.Quests.S24.DataReceiver.Holding\")))"))

(event OnHeldObjectDropped(FortHeldObject)
  ;[1 1 1] Set variables to track player interaction; report quest progress
  (Variables|Default|SetHeld false)
  (Transformation|SetActorRotation (Transformation|GetActorRotation (HeldObjects|Instigator|GetInstigator (Variables|Default|GetFortHeldObject))))
  (CallFunction|ControlIcons true)
  (GameplayEffects|RemoveActiveEffectswithGrantedTags (Ability|GetAbilitySystemComponent (Actor|GetOwner)) "(GameplayTags=((TagName=\"Athena.Quests.S24.DataReceiver.Holding\")))"))

(event OnHeldObjectThrown(FortHeldObject)
  ;[1 1 1] Set variables to track player interaction; report quest progress
  (Variables|Default|SetHeld false)
  (Transformation|SetActorRotation (Transformation|GetActorRotation (HeldObjects|Instigator|GetInstigator (Variables|Default|GetFortHeldObject))))
  (CallFunction|ControlIcons true)
  (GameplayEffects|RemoveActiveEffectswithGrantedTags (Ability|GetAbilitySystemComponent (Actor|GetOwner)) "(GameplayTags=((TagName=\"Athena.Quests.S24.DataReceiver.Holding\")))"))

(event EventBeginPlay
  ;[1 1 1] Begin Play
  (|CallParentFunction)
  (Utilities|FlowControl|DelayUntilNextTick)
  (bind _asfort_player_pawn (Utilities|Casting|CastToFortPlayerPawn (Actor|GetOwner)))
  (Creative|GameplayEffects|SimpleApplyGameplayEffectToActor (Variables|Default|GetCarry%20GE) _asfort_player_pawn)
  (Variables|/Script/Engine.MovementComponent|SetVelocity "0, 0, 0" (Variables|Default|GetFortProjectileMovement))
  (Ability|SendGameplayEventtoActor _asfort_player_pawn "(TagName=\"Event.Quest.Narrative.DataReceiver.Visible\")" (Utilities|Struct|MakeGameplayEventData "(TagName=\"\")" _asfort_player_pawn self))
  (if (not (Networking|IsServer))
    (CallFunction|ServerStartPickup)))

(event Custom|ServerStartPickup
  (bind _fortheldobject (Variables|Default|GetFortHeldObject))
  ;[1 1 1] Server Start Pickup
  (bind _asfort_player_pawn (Utilities|Casting|CastToFortPlayerPawn (Actor|GetOwner)))
  (HeldObjects|Behavior|PickupHeldObject _fortheldobject _asfort_player_pawn)
  ;[1 1 1] Bind to object attached, hide icons and start poi overlap checks
  (EventDispatchers|BindEventtoOnHeldObjectAttached _fortheldobject (delegate Custom|Rotate%20Sign)))

(event Custom|Rotate%20Sign
  ;[1 1 1] Bind to object attached, hide icons and start poi overlap checks
  (CallFunction|MulticastSetLocation)
  (CallFunction|ControlIcons false)
  (Variables|Default|SetHeld true)
  (EventDispatchers|UnbindEventfromOnHeldObjectAttached (Variables|Default|GetFortHeldObject) (delegate Custom|Rotate%20Sign)))

(event Custom|MulticastSetLocation
  ;[1 1 1] Sets Object Location while carried
  (Transformation|SetActorRelativeRotation (Variables|Default|GetDesiredRotation))
  (Transformation|SetActorRelativeLocation (Variables|Default|GetDesiredLocation)))

(event Custom|ClearTimerAfterDelay
  ;[1 1 1] Stops the poi checks after a delay.
  (Utilities|FlowControl|Delay (Variables|Default|GetThrownTimerClearDelay))
  (if (not (Variables|Default|GetHeld))
    (Utilities|Time|ClearTimerbyFunctionName 0 "POI Overlap Check")))

(event Interaction|EventBlueprintOnInteract
  (params (InteractingPawn :type Fort%20Pawn%20Object%20Reference) (InteractionBeingAttempted :type EInteractionBeingAttempted%20Enum :default "FirstInteraction"))
  ;[1 1 1] On Interact
  (switch Utilities|FlowControl|Switch|SwitchonEInteractionBeingAttempted InteractionBeingAttempted
    (:FirstInteraction
      _)
    (:SecondInteraction
      _)
    (:AllInteraction
      _))
  (bind _asfort_player_pawn (Utilities|Casting|CastToFortPlayerPawn InteractingPawn))
  (Variables|/Script/Engine.MovementComponent|SetVelocity "0, 0, 0" (Variables|Default|GetFortProjectileMovement))
  (if (Networking|IsServer)
    (HeldObjects|Behavior|PickupHeldObject (Variables|Default|GetFortHeldObject) _asfort_player_pawn)
    ;[1 1 1] Bind to object attached, hide icons and start poi overlap checks
    ;[1 1 1] Server Start Pickup
    (EventDispatchers|BindEventtoOnHeldObjectAttached (Variables|Default|GetFortHeldObject) (delegate Custom|Rotate%20Sign))))

(event Custom|Control%20Icons
  (params (ShowIcons :type Boolean))
  ;[1 1 1] Call to show or hide icons, executed on client
  (CallFunction|OnRepShowIcon ShowIcons))

(event OnLinkedActorDisconnected(FortLinkToActor)
  (if (not (Variables|Default|GetHeld))
    (Variables|ProjectileMovement|SetRepStartMoving 1)))

(event Custom|ClientEnd
  (if (not (Networking|IsServer))
    (Components|Activation|Activate (Variables|Default|GetDeathExplosion) false)
    (Development|SetHiddeninGame (Variables|Default|GetStaticMesh) true false)))

(event OnComponentHit(Capsule)
  (params (HitComponent :type Primitive%20Component%20Object%20Reference) (OtherActor :type Actor%20Object%20Reference) (OtherComp :type Primitive%20Component%20Object%20Reference) (NormalImpulse :type Vector :default "0, 0, 0") (Hit :type Hit%20Result%20Structure%20%28by%20ref%29))
  ;[1 1 1] Play sound que on hit
  (Utilities|Casting|CastToFortPawn OtherActor
    (:CastFailed
      (Audio|PlaySoundatLocation (Variables|Default|GetSound) (Transformation|GetActorLocation)))))

(fn BlueprintCanInteract
  (params (InteractingPawn :type Fort%20Pawn%20Object%20Reference) (InteractionBeingAttempted :type EInteractionBeingAttempted%20Enum :default "FirstInteraction") (InteractionType :type TInteractionType%20Enum :default "IT_NoInteraction"))
  (Utilities|IsValid InteractingPawn
    (:Is%20Valid
      (bind _returnvalue (HeldObjects|Interaction|CanInteract (Variables|Default|GetFortHeldObject) 0))
      (bind _asfort_player_pawn (Utilities|Casting|CastToFortPlayerPawn (Actor|GetOwner)))
      (return (Math|Boolean|ANDBoolean (== (Utilities|Casting|CastToFortPlayerPawn InteractingPawn) _asfort_player_pawn) _returnvalue (not (GameplayTags|HasMatchingGameplayTag InteractingPawn "(TagName=\"Gameplay.Status.HidingInProp\")")) (not (GameplayTags|HasAnyMatchingGameplayTags InteractingPawn (Variables|Interact|GetTC_BlockPickup))))))))

(fn BlueprintGetInteractionString
  (params (InteractingPawn :type Fort%20Pawn%20Object%20Reference) (InteractionBeingAttempted :type EInteractionBeingAttempted%20Enum :default "FirstInteraction"))
  (return (Variables|Default|GetInteractString)))

(fn OnRep_bIsInPOI
  (if (not (Variables|Default|GetbIsInPOI))
    (CallFunction|ClientEnd)))

(fn OnRep_Held
  (if (Variables|Default|GetHeld)
    (Utilities|Time|SetTimerbyFunctionName 0 "POI Overlap Check" 0.1 true)
    (if (not (Networking|IsServer))
      (bind _asfort_player_pawn (Utilities|Casting|CastToFortPlayerPawn (Actor|GetOwner)))
      (Ability|SendGameplayEventtoActor _asfort_player_pawn "(TagName=\"Event.Quest.Narrative.DataReceiver.PickedUp\")" (Utilities|Struct|MakeGameplayEventData "(TagName=\"\")" _asfort_player_pawn)))
    (else
      (CallFunction|ClearTimerAfterDelay)
      (if (not (Networking|IsServer))
        (bind _asfort_player_pawn_2 (Utilities|Casting|CastToFortPlayerPawn (Actor|GetOwner)))
        (Ability|SendGameplayEventtoActor _asfort_player_pawn_2 "(TagName=\"Event.Quest.Narrative.DataReceiver.Dropped\")" (Utilities|Struct|MakeGameplayEventData "(TagName=\"\")" _asfort_player_pawn_2))))))

(fn OnRep_ShowIcon
  (params (ShowIcons :type Boolean))
  (bind _fortquesticon (Variables|Default|GetFortQuestIcon))
  (MiniMapData|SetViewDistanceRadius _fortquesticon (MiniMapData|GetViewDistanceRadius _fortquesticon))
  (MiniMapData|SetVisibility _fortquesticon ShowIcons)
  (Variables|Default|SetOwnerCanInteract ShowIcons))

(fn POIOverlapCheck
  (bind (_outactors _returnvalue) (Collision|CapsuleOverlapActors (Transformation|GetActorLocation) 65.0 65.0 (Utilities|Array|MakeArray "ObjectTypeQuery1" "ObjectTypeQuery2") "/BattlePassS24/Items/QuestActors/Narrative/Part01/DataReceiver/BP_S24_PropQuestActor_DataReceiver_Place.BP_S24_PropQuestActor_DataReceiver_Place_C"))
  (if _returnvalue
    (bind _asfort_player_pawn (Utilities|Casting|CastToFortPlayerPawn (Actor|GetOwner)))
    (bind _asfort_player_controller (Utilities|Casting|CastToFortPlayerController (Pawn|GetController _asfort_player_pawn)))
    (HeldObjects|Behavior|DropHeldObject (Variables|Default|GetFortHeldObject))
    (Utilities|Time|ClearTimerbyFunctionName 0 "POI Overlap Check")
    (Quest|SendComplexCustomStatEvent (Quests|GetQuestManager _asfort_player_controller "Athena") 0 "(GameplayTags=)" "(GameplayTags=)" 1)
    (CallFunction|ControlIcons false)
    (Development|SetHiddeninGame (Variables|Default|GetStaticMesh) true false)))

(fn SetAttachDefaultTransform)

(fn ShouldSignBeEnabled
  (bind _outxy (DataTable|EvaluateCurveTableRow "/Game/Athena/Balance/DataTables/AthenaGameData.AthenaGameData" "Default.QuestInteractable.GiantPinkBear.Enabled")
    (:RowFound
      (if (>= _outxy 1.0)
        (return true)
        (else
          (return false))))
    (:RowNotFound
      (return false))))

(fn ConstructionScript
  (|CallParentFunction)
  (Effects|Components|ParticleSystem|SetTemplate (Variables|Default|GetDeathExplosion) (Variables|Default|GetBearPopEffect)))
