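; Season 24 narrative quest prop for a data receiver pickup that players interact with to receive a carried version.
; Interacting spawns the held-object actor, stores it per player in a map, and hides receiver and quest icon only for that player via remote-side, splitscreen-aware local controller checks.
; Quest-ready sends a delayed gameplay event to show the receiver.
; Reaching the objective count expires that player's carried actor through a short lifespan.
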(event Interaction|EventBlueprintOnInteract
  (params (InteractingPawn :type Fort%20Pawn%20Object%20Reference) (InteractionBeingAttempted :type EInteractionBeingAttempted%20Enum :default "FirstInteraction"))
  (bind (_returnvalue_x _returnvalue_y) (Math|Vector|BreakVector (Transformation|GetActorLocation)))
  (bind (_returnvalue_roll _returnvalue_pitch _returnvalue_yaw) (Math|Rotator|BreakRotator (Transformation|GetActorRotation)))
  ;[0 0 0] OnInteract override - Spawn the carried actor, set its colision, then hide the quest marker
  (|CallParentFunction InteractingPawn InteractionBeingAttempted)
  (bind _asfort_player_pawn (Utilities|Casting|CastToFortPlayerPawn InteractingPawn))
  (CallFunction|SendGameplayEventtoInstigator _asfort_player_pawn)
  (bind _returnvalue (Game|SpawnActorfromClass "/BattlePassS24/Items/QuestActors/Narrative/Part01/DataReceiver/BGA_HeldObject_Quest_DataReceiver_Carry.BGA_HeldObject_Quest_DataReceiver_Carry_C" (Math|Transform|MakeTransform (Math|Vector|MakeVector (- _returnvalue_x -7.0) (- _returnvalue_y -14.0)) (Math|Rotator|MakeRotator _returnvalue_roll _returnvalue_pitch (- _returnvalue_yaw 7.0))) "Undefined" "OverrideRootScale" InteractingPawn 0 0))
  (Utilities|Map|Add (Variables|Default|GetPlayersToCarriedActors) _asfort_player_pawn _returnvalue)
  (CallFunction|ReplicatedHideIcon _asfort_player_pawn))

(event Custom|SendGameplayEventToInstigator
  (params (InstigatingPlayerPawn :type Fort%20Player%20Pawn%20Object%20Reference))
  ;[0 0 0] Sends a gameplay event on the interacting player's client to hide the data receiver
  (Networking|SwitchHasAuthority
    (:Remote
      (bind _output (CallFunction|IsValidLocalPlayerController (Fortnite|GetFortPlayerControllerfromActor InstigatingPlayerPawn)))
      (if _output
        (Ability|SendGameplayEventtoActor InstigatingPlayerPawn "(TagName=\"Event.Quest.Narrative.DataReceiver.Hidden\")" (Utilities|Struct|MakeGameplayEventData "(TagName=\"\")" 0 self))))))

(event OnCustomUpdateObjective(ConsolidatedQuestComponent)
  (params (PlayerController :type Fort%20Player%20Controller%20Object%20Reference) (AchievedCount :type Integer) (RequiredCount :type Integer))
  (bind _returnvalue (Game|Player|GetPlayerPawn PlayerController))
  (bind _playerstocarriedactors (Variables|Default|GetPlayersToCarriedActors))
  (bind (_value _returnvalue_2) (Utilities|Map|Find _playerstocarriedactors _returnvalue))
  ;[0 0 0] Destroy spawned actors for a player after their quest is completed
  (if (>= AchievedCount RequiredCount)
    (if _returnvalue_2
      (Utilities|IsValid _value
        (:Is%20Valid
          (Actor|SetLifeSpan _value 0.2))))))

(event OnQuestReady(ConsolidatedQuestComponent)
  (params (PlayerController :type Fort%20Player%20Controller%20Object%20Reference))
  ;[0 0 0] Show the data receiver when the quest is ready
  (Utilities|FlowControl|Delay 1.0)
  (Ability|SendGameplayEventtoActor (Game|Player|GetPlayerPawn PlayerController) "(TagName=\"Event.Quest.Narrative.DataReceiver.Visible\")" (Utilities|Struct|MakeGameplayEventData "(TagName=\"\")" (Game|Player|GetPlayerPawn PlayerController) self)))

(event Custom|Replicated_HideIcon
  (params (InteractingPawn :type Fort%20Pawn%20Object%20Reference))
  ;[0 0 0] Quest marker visibility handling
  ;;[0.1804 0.1804 0.1804] Set visibility of the icon to false on the client that triggered the interaction
  (Networking|SwitchHasAuthority
    (:Remote
      (bind _output (CallFunction|IsValidLocalPlayerController (Fortnite|GetFortPlayerControllerfromActor InteractingPawn)))
      (if _output
        (MiniMapData|SetVisibility (Variables|BuildingGameplayActorPropQuest|GetQuestIconComponent) false)))))

(event Custom|ReplicatedShowIcons
  (params (InteractingPawn :type Fort%20Pawn%20Object%20Reference))
  ;[0 0 0] Quest marker visibility handling
  ;;[0.1804 0.1804 0.1804] Set visibility of the icon to true on the client that triggered the interaction
  (Networking|SwitchHasAuthority
    (:Remote
      (bind _output (CallFunction|IsValidLocalPlayerController (Fortnite|GetFortPlayerControllerfromActor InteractingPawn)))
      (if _output
        (MiniMapData|SetVisibility (Variables|BuildingGameplayActorPropQuest|GetQuestIconComponent) true)))))

(fn IsValidLocalPlayerController
  (params (InstigatingPlayerController :type Fort%20Player%20Controller%20Object%20Reference))
  ;[0 0 0] Returns true on the executing client if the instigator controller is locally controlled. Returns false if running on the server
  (if (not (Networking|IsServer))
    (Utilities|IsValid InstigatingPlayerController
      (:Is%20Valid
        (if (PlayerController|IsinSplitscreen InstigatingPlayerController)
          (for _array_element (Fortnite|GetLocalFortPlayerControllers true)
            (if (== _array_element InstigatingPlayerController)
              (return true)))
          (return false)
          (else
            (return true))))
      (:Is%20Not%20Valid
        _)))
  (return false))

(fn ConstructionScript
  (|CallParentFunction))
