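; Season 24 quest prop actor for placing a data receiver, a child of a parent prop quest actor.
; The construction script calls the parent, applies the screen-off material to slot 1 of the disabled bot beacon mesh and sets its relative transform from a data receiver variable.
; Local interaction runs the parent logic, then hides that disabled beacon.
; The visual-testing reset only calls the parent before an unconnected authority switch, noted as disabled.
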(event EventBeginPlay
  (|CallParentFunction) ; This node is disabled and will not be called.\nDrag off pins to build functionality.
)

(event Collision|EventActorBeginOverlap
  (params (OtherActor :type Actor%20Object%20Reference))
  (|CallParentFunction OtherActor) ; This node is disabled and will not be called.\nDrag off pins to build functionality.
)

(event EventTick
  (params (DeltaSeconds :type Float%20%28single-precision%29))
  (|CallParentFunction DeltaSeconds) ; This node is disabled and will not be called.\nDrag off pins to build functionality.
)

(event EventResetforVisualTesting
  (params (PlayerController :type Fort%20Player%20Controller%20Object%20Reference))
  ;[1 0.2396 0.2396] This logic calls even if the prop is not in visual test mode. Disabling
  (|CallParentFunction PlayerController)
  (Networking|SwitchHasAuthority))

(fn BlueprintOnLocalInteract
  (params (InteractingPawn :type Fort%20Player%20Pawn%20Object%20Reference) (InteractionBeingAttempted :type EInteractionBeingAttempted%20Enum :default "FirstInteraction"))
  (bind _returnvalue (|CallParentFunction InteractingPawn InteractionBeingAttempted))
  (Rendering|SetVisibility (Variables|Default|GetBotBeacon_Disabled) false false)
  (return _returnvalue))

(fn PostInteractionVFX&Audio
  (params (PlayerController :type Fort%20Player%20Controller%20Object%20Reference))
  (|CallParentFunction PlayerController))

(fn ConstructionScript
  (|CallParentFunction)
  (Rendering|Material|SetMaterial (Variables|Default|GetBotBeacon_Disabled) 1 (Variables|DataReceiver|GetScreenMaterial_Off))
  (Transformation|SetRelativeTransform (Variables|Default|GetBotBeacon_Disabled) (Variables|DataReceiver|GetDataReceiverTransform) false false))
